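# Post-process AnimBlueprint for the male medium Grim head that drives the dyn_neck_01 bone from head rotation.
# Each update, the EventGraph stores the head bone's delta from ref pose relative to neck_02 on the owning mesh.
# The AnimGraph replaces dyn_neck_01's bone-space rotation with that value, halving yaw while keeping roll and pitch.
# Two Constraint nodes for faceAttach and dyn_neck_01 sit unconnected to the output pose.
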AnimBlueprint /BRCosmetics/Characters/Player/Male/Medium/Heads/M_MED_Grim_Head/Meshes/Male_Medium_Grim_Head_AnimBP_PostProcess
target_skeleton=/Game/Characters/Player/Common/Fortnite_Base_Head/Fortnite_Base_Head_Export_Skeleton
graphs [
 AnimGraph {nodes [
   "Constraint - faceAttach (P)"(Alpha=1, ConstraintWeights_0=1) {
    alphaBoolBlend {blendOption=Linear}
    boneToModify {boneName=faceAttach}
    constraintSetup [{targetBone {boneName=dyn_neck_01}}]
    constraintWeights [1]
   }
   "Constraint - dyn_neck_01 (P)"(Alpha=1, ConstraintWeights_0=1) {
    alphaBoolBlend {blendOption=Linear}
    boneToModify {boneName=dyn_neck_01}
    constraintSetup [{targetBone {boneName=neck_02}}]
    constraintWeights [1]
   }
   n1 = BreakRotator(InRot=Get head rot() -> (head rot:Rotator))
   n2 = "Transform (Modify) Bone - Bone: dyn_neck_01"(
    ComponentPose=Local To Component(LocalPose=Input Pose())
    Alpha=1
    Translation="0,0,0"
    Rotation=MakeRotator(Roll=@n1, Pitch=@n1.Pitch, Yaw=float * float(A=@n1.Yaw, B=0.5))
    Scale="1,1,1"
   ) {rotationMode=BMM_Replace, rotationSpace=BCS_BoneSpace, alphaBoolBlend {blendOption=Linear}, boneToModify {boneName=dyn_neck_01}}
   Output Pose(Result=Component To Local(ComponentPose=@n2)) {layerGroup=None}
  ]}
 EventGraph {nodes [
   Event BlueprintUpdateAnimation()
   n1 = Get Skel_Mesh()
   > Set head rot(head rot:Rotator=BreakTransform(InTransform=GetDeltaTransformFromRefPose(self=@n1, BoneName=head, BaseName=neck_02)).Rotation) -> (Output_Get:Rotator)
   Event BlueprintInitializeAnimation()
   > Set Skel_Mesh(Skel_Mesh=GetOwnerSkeletalMeshComponent())
  ]}
]
variables [
 head rot {type=Rotator}
 Skel_Mesh {type=SkeletalMeshComponent}
]
